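# Animation Blueprint for the Pirate 02 backpack cosmetic that adds secondary physics motion to its rope and cup bones.
# Starting from the ref pose, it additively offsets backpack_root in bone space by BackpackTranslation and BackpackRotation, neither declared here.
# The result feeds chained Anim Dynamics nodes for dyn_rope and dyn_cup, with angular limits, high damping and scaled component motion.
# Only the cup enables a gravity override, and the EventGraph is empty.
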AnimBlueprint /BRCosmetics/Accessories/FORT_Backpacks/M_MED_Pirate_02_Backpack/Meshes/M_MED_Pirate_02_Backpack_AnimBlueprint
target_skeleton=&/M_MED_Pirate_02_Backpack_Skeleton
graphs [
 AnimGraph {nodes [
   n1 = "Transform (Modify) Bone - Bone: backpack_root"(
    ComponentPose=Local To Component(LocalPose=Local Space Ref Pose())
    Alpha=1
    Translation=Get BackpackTranslation() -> (BackpackTranslation:Vector)
    Rotation=Get BackpackRotation() -> (BackpackRotation:Rotator)
    Scale="1,1,1"
   ) {
    rotationMode=BMM_Additive
    rotationSpace=BCS_BoneSpace
    translationMode=BMM_Additive
    translationSpace=BCS_BoneSpace
    alphaBoolBlend {blendOption=Linear}
    boneToModify {boneName=backpack_root}
   }
   n2 = Anim Dynamics - Bone: dyn_rope(ComponentPose=@n1, Alpha=1, ExternalForce="0,0,0") {
    angularDampingOverride=0.99
    bGravityOverrideInSimSpace=true
    bOverrideAngularDamping=true
    alphaBoolBlend {blendOption=Linear}
    boundBone {boneName=dyn_rope}
    componentLinearAccScale=<0.7 0.7 0.7>
    componentLinearVelScale=<0.7 0.7 0.7>
    physicsBodyDefinitions [{
      sphereCollisionRadius=0
      boundBone {boneName=dyn_rope}
      boxExtents=<4 9 5>
      constraintSetup {angularTargetAxis=AxisY, twistAxis=AxisY, angularLimitsMax=<45 0 45>, angularLimitsMin=<-45 0 -10>, angularTarget=<0 0 0>}
      localJointOffset=<0 -0.5 0>
     }]
   }
   n3 = Anim Dynamics - Bone: dyn_cup(ComponentPose=@n2, Alpha=1, ExternalForce="0,0,0") {
    angularDampingOverride=0.99
    bGravityOverrideInSimSpace=true
    bOverrideAngularDamping=true
    bUseGravityOverride=true
    alphaBoolBlend {blendOption=Linear}
    boundBone {boneName=dyn_cup}
    componentLinearAccScale=<0.7 0.7 0.7>
    componentLinearVelScale=<0.7 0.7 0.7>
    gravityOverride {z=-980}
    physicsBodyDefinitions [{
      sphereCollisionRadius=0
      boundBone {boneName=dyn_cup}
      boxExtents=<4 9 5>
      constraintSetup {angularTargetAxis=AxisY, twistAxis=AxisY, angularLimitsMax=<0 0 45>, angularLimitsMin=<0 0 0>, angularTarget=<0 0 0>}
      localJointOffset=<0 -0.5 0>
     }]
   }
   Output Pose(Result=Component To Local(ComponentPose=@n3)) {layerGroup=None}
  ]}
 EventGraph {links [], nodes []}
]
variables []
